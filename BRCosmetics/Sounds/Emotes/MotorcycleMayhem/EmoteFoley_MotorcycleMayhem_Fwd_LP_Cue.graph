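# Foley sound cue for the forward-movement loop of the Motorcycle Mayhem emote, using emote foley attenuation, class and concurrency.
# A mixer blends a quieter looping dirt-tire layer with the forward-start sound, which runs through Doppler, looping and an envelope that fades in volume.
# Continuous modulators on both layers tie pitch and absolute volume to the EmoteMoveSpeed parameter, so the sound follows movement speed.
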SoundCue /BRCosmetics/Sounds/Emotes/MotorcycleMayhem/EmoteFoley_MotorcycleMayhem_Fwd_LP_Cue
root=Mixer_1
nodes [
 {id=Doppler_3, children [WavePlayer_8]}
 {
  id=Enveloper_0
  children [ModulatorContinuous_3]
  properties {pitchCurve {editorCurveData {keys [{time=0}]}}, volumeCurve {editorCurveData {keys [<at 0 0>,<at 1 1>]}}}
 }
 {id=Looping_1, children [Doppler_3]}
 {id=Mixer_1, children [ModulatorContinuous_5, Enveloper_0], properties {inputVolume [0.3,1]}}
 {
  id=ModulatorContinuous_3
  children [Looping_1]
  properties {
   pitchModulationParams {minOutput=0.9, parameterName=EmoteMoveSpeed}
   volumeModulationParams {paramMode=MPM_Abs, parameterName=EmoteMoveSpeed}
  }
 }
 {
  id=ModulatorContinuous_5
  children [WavePlayer_4]
  properties {
   pitchModulationParams {minOutput=0.9, parameterName=EmoteMoveSpeed}
   volumeModulationParams {paramMode=MPM_Abs, parameterName=EmoteMoveSpeed}
  }
 }
 {
  id=WavePlayer_4
  properties {bLooping=true, soundWaveAssetPtr=&/Emote_MotorcycleMayhem_Tire_Dirt_LP}
 }
 {id=WavePlayer_8, properties {soundWaveAssetPtr=&/Emote_MotorcycleMayhem_Fwd_Start}}
]
properties {
 attenuationSettings=&Fort_Emotes_Athena_Att
 audioPropertiesSheet=/BRCosmetics/Sounds/Settings/PropertySheets/Emotes/APS_Cosmetic_Emote_Foley
 maxDistance=1600
 soundClassObject=/Game/Sounds/FortSoundClassesAndModes/Fort_Foley_Emote
 volumeMultiplier=1.1
 concurrencyOverrides {maxCount=3}
 concurrencySet [/Game/Sounds/Concurrency_Rules/Athena_EmoteFoley_Conc]
 soundSubmixSends [{minSendLevel=0, soundSubmix=&Fortnite_AudioAnalysis_Music}]
}
