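# Looping engine idle foley cue for the Motorcycle Mayhem emote.
# A continuous modulator reads EmoteMoveSpeed to fade the idle loop's volume down as speed rises, feeding a mixer whose second input is empty.
# Emote foley attenuation, sound class and concurrency apply, with play-when-silent virtualization and a submix send for music audio analysis.
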SoundCue /BRCosmetics/Sounds/Emotes/MotorcycleMayhem/EmoteFoley_MotorcycleMayhem_Rev_LP_Cue
root=Mixer_0
nodes [
 {id=Mixer_0, children [ModulatorContinuous_5, null], properties {inputVolume [0.6,1]}}
 {
  id=ModulatorContinuous_5
  children [WavePlayer_13]
  properties {
   pitchModulationParams {minOutput=1, parameterName=EmoteMoveSpeed}
   volumeModulationParams {maxOutput=0, minOutput=1, parameterName=EmoteMoveSpeed}
  }
 }
 {
  id=WavePlayer_13
  properties {bLooping=true, soundWaveAssetPtr=&/Emote_MotorcycleMayhem_Rev_Idle_LP}
 }
]
properties {
 attenuationSettings=&Fort_Emotes_Athena_Att
 audioPropertiesSheet=/BRCosmetics/Sounds/Settings/PropertySheets/Emotes/APS_Cosmetic_Emote_Foley
 maxDistance=1600
 soundClassObject=/Game/Sounds/FortSoundClassesAndModes/Fort_Foley_Emote
 virtualizationMode=PlayWhenSilent
 volumeMultiplier=1.1
 concurrencyOverrides {maxCount=3}
 concurrencySet [/Game/Sounds/Concurrency_Rules/Athena_EmoteFoley_Conc]
 soundSubmixSends [{minSendLevel=0, soundSubmix=&Fortnite_AudioAnalysis_Music}]
}
